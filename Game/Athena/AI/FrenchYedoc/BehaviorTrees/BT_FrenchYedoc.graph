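# Behavior tree for FrenchYedoc AI bots that prioritizes moving to a beacon over standard Phoebe bot gameplay.
# A ReachBeacon evaluator service on the root selector periodically feeds blackboard keys for execution status and target.
# When that status is at least ExecutionPending, the bot moves to the beacon target with strafing allowed and partial paths disallowed.
# Otherwise, or if the move fails, the selector runs BT_Phoebe_Gameplay as a subtree.
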BehaviorTree /Game/Athena/AI/FrenchYedoc/BehaviorTrees/BT_FrenchYedoc
blackboard=&/BB_FrenchYedoc
root {
 name=Selector
 class=BTComposite_Selector
 children [
  MoveTo Beacon {
   class=FortAthenaBTTask_BotMoveTo
   decorators [Can Reach a Beacon? {
     class=BTDecorator_Blackboard
     properties {
      arithmeticOperation=GreaterOrEqual
      intValue=3
      nodeName=Can Reach a Beacon?
      stringValue=EExecutionStatus::ExecutionPending
      blackboardKey {selectedKeyName=AIEvaluator_ReachBeacon_ExecutionStatus}
     }
    }]
   properties {
    movementResultKeyName=AIEvaluator_ReachBeacon_MovementState
    nodeName=MoveTo Beacon
    acceptableRadius {defaultValue=30}
    allowRandomWobble {defaultValue=false}
    bAllowPartialPath {defaultValue=false}
    bAllowStrafe {defaultValue=true}
    blackboardKey {selectedKeyName=AIEvaluator_ReachBeacon_Target}
    observedBlackboardValueTolerance {defaultValue=50}
   }
  }
  Run Behavior {
   class=BTTask_RunBehavior
   decorators []
   properties {behaviorAsset=/Game/Athena/AI/Phoebe/BehaviorTrees/BT_Phoebe_Gameplay, nodeName=Run Behavior}
  }
 ]
 decorators []
 properties {
  children [
   {
    childComposite=null
    childTask=&self:FortAthenaBTTask_BotMoveTo_0
    decorators [&self:BTDecorator_Blackboard_2]
   }
   {childComposite=null, childTask=&self:BTTask_RunBehavior_0, decorators []}
  ]
  services [&self:FortAthenaBTService_AIEvaluator_0]
 }
 services [Evaluating... ReachBeacon {
   class=FortAthenaBTService_AIEvaluator
   properties {
    interval=0.5
    nodeName=Evaluating... ReachBeacon
    aIEvaluatorClass {defaultValue=/Script/FortniteAIServer.FortAthenaAIBotEvaluator_ReachBeacon}
   }
  }]
}
